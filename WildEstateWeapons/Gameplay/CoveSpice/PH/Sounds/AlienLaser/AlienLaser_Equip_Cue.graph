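# Sound cue for equipping the AlienLaser weapon.
# A Random node picks between two equally weighted equip waves, and a Modulator at its default settings wraps the result.
# The cue plays as weapon foley with reload attenuation, reduced volume and slightly raised pitch.
# It overrides concurrency and has its debug flag switched on.
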SoundCue /WildEstateWeapons/Gameplay/CoveSpice/PH/Sounds/AlienLaser/AlienLaser_Equip_Cue
root=Modulator_0
nodes [
 {id=Modulator_0, children [Random_1]}
 {id=Random_1, children [WavePlayer_2,WavePlayer_3], properties {weights [1,1]}}
 {id=WavePlayer_2, properties {soundWaveAssetPtr=&/AlienLaser_Equip_01}}
 {id=WavePlayer_3, properties {soundWaveAssetPtr=&/AlienLaser_Equip_02}}
]
properties {
 attenuationSettings=&Fort_Weapon_Reload_Att
 bDebug=true
 bOverrideConcurrency=true
 maxDistance=2250
 pitchMultiplier=1.1
 soundClassObject=&Fort_WeaponFoley
 volumeMultiplier=0.351
}
