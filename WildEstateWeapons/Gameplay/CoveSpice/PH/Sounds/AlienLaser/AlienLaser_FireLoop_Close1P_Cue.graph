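# Sound cue for the first-person, close-range looping fire sound of the AlienLaser weapon.
# Two looping waves mix with a pitch-randomized one-shot branch, which layers a random fire shot at fixed lowered pitch with a random Layer B shot.
# The mix routes through first-person projectile attenuation and a first-person weapon fire sound class, under gunfire concurrency.
# The Layer B mixer has an empty second input, and the debug flag is on.
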SoundCue /WildEstateWeapons/Gameplay/CoveSpice/PH/Sounds/AlienLaser/AlienLaser_FireLoop_Close1P_Cue
root=SoundClass_1
nodes [
 {
  id=Attenuation_1
  children [Mixer_0]
  properties {attenuationSettings=&Fort_Weapon_Projectile_FirstPerson_Athena_Att}
 }
 {id=Mixer_0, children [Modulator_0,WavePlayer_1,WavePlayer_3], properties {inputVolume [0.7,0.9,0.5]}}
 {id=Mixer_1, children [Modulator_1,Mixer_2], properties {inputVolume [1,1]}}
 {id=Mixer_2, children [Random_9,null], properties {inputVolume [1,0.7]}}
 {id=Modulator_0, children [Mixer_1], properties {pitchMax=1.1, pitchMin=0.9}}
 {id=Modulator_1, children [Random_2], properties {pitchMax=0.9, pitchMin=0.9}}
 {id=Random_2, children [WavePlayer_9, WavePlayer_10, WavePlayer_11, WavePlayer_12], properties {weights [1,1,1,1]}}
 {
  id=Random_9
  children [WavePlayer_39, WavePlayer_40, WavePlayer_41, WavePlayer_42, WavePlayer_43, WavePlayer_44]
  properties {weights [1,1,1,1,1,1]}
 }
 {
  id=SoundClass_1
  children [Attenuation_1]
  properties {soundClassOverride=&Fort_WeaponRanged_Fire_1P_PSM}
 }
 {
  id=WavePlayer_1
  properties {bLooping=true, soundWaveAssetPtr=&/AlienLaser_FireLoop_Close_New_02}
 }
 {id=WavePlayer_10, properties {soundWaveAssetPtr=&/AlienLaser_Fire_Close_02}}
 {id=WavePlayer_11, properties {soundWaveAssetPtr=&/AlienLaser_Fire_Close_03}}
 {id=WavePlayer_12, properties {soundWaveAssetPtr=&/AlienLaser_Fire_Close_04}}
 {
  id=WavePlayer_3
  properties {bLooping=true, soundWaveAssetPtr=&/AlienLaser_FireLoop_Close_New_05}
 }
 {id=WavePlayer_39, properties {soundWaveAssetPtr=&/AlienLaser_Fire_Close_Layer_B_01}}
 {id=WavePlayer_40, properties {soundWaveAssetPtr=&/AlienLaser_Fire_Close_Layer_B_02}}
 {id=WavePlayer_41, properties {soundWaveAssetPtr=&/AlienLaser_Fire_Close_Layer_B_03}}
 {id=WavePlayer_42, properties {soundWaveAssetPtr=&/AlienLaser_Fire_Close_Layer_B_04}}
 {id=WavePlayer_43, properties {soundWaveAssetPtr=&/AlienLaser_Fire_Close_Layer_B_05}}
 {id=WavePlayer_44, properties {soundWaveAssetPtr=&/AlienLaser_Fire_Close_Layer_B_06}}
 {id=WavePlayer_9, properties {soundWaveAssetPtr=&/AlienLaser_Fire_Close_01}}
]
properties {bDebug=true, maxDistance=1800.0001, volumeMultiplier=0.6, concurrencySet [&Weapon_Gunfire_FirstPerson_Conc]}
