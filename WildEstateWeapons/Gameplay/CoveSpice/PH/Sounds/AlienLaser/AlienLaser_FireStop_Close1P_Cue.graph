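# First-person, close-range sound cue for the alien laser weapon's fire-stop event.
# A mixer layers one of three equally weighted random C-layer variations, boosted in volume, with a quieter fixed F layer.
# The mix passes through first-person projectile attenuation into a first-person ranged-fire sound class override.
# The cue plays at low overall volume under first-person gunfire concurrency and has its debug flag enabled.
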SoundCue /WildEstateWeapons/Gameplay/CoveSpice/PH/Sounds/AlienLaser/AlienLaser_FireStop_Close1P_Cue
root=SoundClass_1
nodes [
 {
  id=Attenuation_1
  children [Mixer_0]
  properties {attenuationSettings=&Fort_Weapon_Projectile_FirstPerson_Athena_Att}
 }
 {id=Mixer_0, children [Random_10, WavePlayer_21], properties {inputVolume [1.5,0.7]}}
 {id=Random_10, children [WavePlayer_45, WavePlayer_46, WavePlayer_47], properties {weights [1,1,1]}}
 {
  id=SoundClass_1
  children [Attenuation_1]
  properties {soundClassOverride=&Fort_WeaponRanged_Fire_1P_PSM}
 }
 {id=WavePlayer_21, properties {soundWaveAssetPtr=&/AlienLaser_Fire_Close_Layer_F_01}}
 {id=WavePlayer_45, properties {soundWaveAssetPtr=&/AlienLaser_Fire_Close_Layer_C_03}}
 {id=WavePlayer_46, properties {soundWaveAssetPtr=&/AlienLaser_Fire_Close_Layer_C_01}}
 {id=WavePlayer_47, properties {soundWaveAssetPtr=&/AlienLaser_Fire_Close_Layer_C_02}}
]
properties {bDebug=true, maxDistance=1800.0001, volumeMultiplier=0.2, concurrencySet [&Weapon_Gunfire_FirstPerson_Conc]}
