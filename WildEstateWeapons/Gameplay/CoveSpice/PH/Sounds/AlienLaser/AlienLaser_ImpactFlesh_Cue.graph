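# Sound cue for the Alien Laser weapon's flesh impact, routed to the impacts sound class.
# A quality-level node sends the first two levels to a mixer with full variant pools and the last two to one with smaller subsets.
# Each mixer layers a random generic laser impact under a louder random flesh impact, both pitched up by modulators.
# Overrides set attenuation falloff and cap concurrency, stopping the oldest instance.
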SoundCue /WildEstateWeapons/Gameplay/CoveSpice/PH/Sounds/AlienLaser/AlienLaser_ImpactFlesh_Cue
root=QualityLevel_0
nodes [
 {id=Mixer_0, children [Modulator_3,Modulator_0], properties {inputVolume [0.8,1]}}
 {id=Mixer_11, children [Modulator_2,Modulator_1], properties {inputVolume [0.8,1]}}
 {id=Modulator_0, children [Random_5], properties {pitchMax=1.3, pitchMin=1.2}}
 {id=Modulator_1, children [Random_4], properties {pitchMax=1.3, pitchMin=1.2}}
 {id=Modulator_2, children [Random_0], properties {pitchMax=1.3, pitchMin=1.2}}
 {id=Modulator_3, children [Random_3], properties {pitchMax=1.3, pitchMin=1.2}}
 {id=QualityLevel_0, children [Mixer_11,Mixer_11,Mixer_0,Mixer_0]}
 {
  id=Random_0
  children [
   WavePlayer_0
   WavePlayer_1
   WavePlayer_2
   WavePlayer_3
   WavePlayer_4
   WavePlayer_5
   WavePlayer_6
   WavePlayer_7
   WavePlayer_8
   WavePlayer_9
   WavePlayer_10
   WavePlayer_11
  ]
  properties {weights [1,1,1,1,1,1,1,1,1,1,1,1]}
 }
 {id=Random_3, children [WavePlayer_24, WavePlayer_25, WavePlayer_26, WavePlayer_27, WavePlayer_28], properties {weights [1,1,1,1,1]}}
 {
  id=Random_4
  children [WavePlayer_21, WavePlayer_22, WavePlayer_23, WavePlayer_29, WavePlayer_30, WavePlayer_31]
  properties {weights [1,1,1,1,1,1]}
 }
 {id=Random_5, children [WavePlayer_32, WavePlayer_33, WavePlayer_34], properties {weights [1,1,1]}}
 {id=WavePlayer_0, properties {soundWaveAssetPtr=&/AlienLaser_Impact_10}}
 {id=WavePlayer_1, properties {soundWaveAssetPtr=&/AlienLaser_Impact_11}}
 {id=WavePlayer_10, properties {soundWaveAssetPtr=&/AlienLaser_Impact_08}}
 {id=WavePlayer_11, properties {soundWaveAssetPtr=&/AlienLaser_Impact_09}}
 {id=WavePlayer_2, properties {soundWaveAssetPtr=&/AlienLaser_Impact_12}}
 {id=WavePlayer_21, properties {soundWaveAssetPtr=&/AlienLaser_ImpactFlesh_01}}
 {id=WavePlayer_22, properties {soundWaveAssetPtr=&/AlienLaser_ImpactFlesh_02}}
 {id=WavePlayer_23, properties {soundWaveAssetPtr=&/AlienLaser_ImpactFlesh_03}}
 {id=WavePlayer_24, properties {soundWaveAssetPtr=&AlienLaser_Impact_10}}
 {id=WavePlayer_25, properties {soundWaveAssetPtr=&AlienLaser_Impact_11}}
 {id=WavePlayer_26, properties {soundWaveAssetPtr=&AlienLaser_Impact_12}}
 {id=WavePlayer_27, properties {soundWaveAssetPtr=&/AlienLaser_Impact_01}}
 {id=WavePlayer_28, properties {soundWaveAssetPtr=&/AlienLaser_Impact_02}}
 {id=WavePlayer_29, properties {soundWaveAssetPtr=&/AlienLaser_ImpactFlesh_04}}
 {id=WavePlayer_3, properties {soundWaveAssetPtr=&AlienLaser_Impact_01}}
 {id=WavePlayer_30, properties {soundWaveAssetPtr=&/AlienLaser_ImpactFlesh_05}}
 {id=WavePlayer_31, properties {soundWaveAssetPtr=&/AlienLaser_ImpactFlesh_06}}
 {id=WavePlayer_32, properties {soundWaveAssetPtr=&AlienLaser_ImpactFlesh_01}}
 {id=WavePlayer_33, properties {soundWaveAssetPtr=&AlienLaser_ImpactFlesh_02}}
 {id=WavePlayer_34, properties {soundWaveAssetPtr=&AlienLaser_ImpactFlesh_03}}
 {id=WavePlayer_4, properties {soundWaveAssetPtr=&AlienLaser_Impact_02}}
 {id=WavePlayer_5, properties {soundWaveAssetPtr=&/AlienLaser_Impact_03}}
 {id=WavePlayer_6, properties {soundWaveAssetPtr=&/AlienLaser_Impact_04}}
 {id=WavePlayer_7, properties {soundWaveAssetPtr=&/AlienLaser_Impact_05}}
 {id=WavePlayer_8, properties {soundWaveAssetPtr=&/AlienLaser_Impact_06}}
 {id=WavePlayer_9, properties {soundWaveAssetPtr=&/AlienLaser_Impact_07}}
]
properties {
 bOverrideAttenuation=true
 bOverrideConcurrency=true
 maxDistance=3400
 soundClassObject=&Fort_Impacts
 volumeMultiplier=1.6
 attenuationOverrides {falloffDistance=3000}
 concurrencyOverrides {maxCount=5, resolutionRule=StopOldest, volumeScale=0.75}
}
