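# Sound cue for alien laser impacts on water, built from placeholder-path reuse of bullet and flare gun water impact waves.
# Each branch mixes a random bullet splash with a quieter, modulated random ember impact.
# A quality node gives the first two levels the full wave pools and the last two smaller three-wave pools.
# Bullet impact attenuation applies, with overridden concurrency that stops the oldest sound.
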SoundCue /WildEstateWeapons/Gameplay/CoveSpice/PH/Sounds/AlienLaser/AlienLaser_ImpactWater_Cue
root=QualityLevel_1
nodes [
 {id=Mixer_0, children [Random_0,Modulator_2], properties {inputVolume [1,0.3]}}
 {id=Mixer_1, children [Random_3,Modulator_0], properties {inputVolume [1,0.3]}}
 {id=Modulator_0, children [Random_4]}
 {id=Modulator_2, children [Random_2]}
 {id=QualityLevel_1, children [Mixer_0,Mixer_0,Mixer_1,Mixer_1]}
 {
  id=Random_0
  children [WavePlayer_0, WavePlayer_1, WavePlayer_2, WavePlayer_3, WavePlayer_4, WavePlayer_5, WavePlayer_6, WavePlayer_7, WavePlayer_19, WavePlayer_20]
  properties {weights [1,1,1,1,1,1,1,1,1,1]}
 }
 {id=Random_2, children [WavePlayer_13, WavePlayer_14, WavePlayer_15, WavePlayer_16, WavePlayer_17], properties {weights [1,1,1,1,1]}}
 {id=Random_3, children [WavePlayer_21, WavePlayer_24, WavePlayer_25], properties {weights [1,1,1]}}
 {id=Random_4, children [WavePlayer_18, WavePlayer_22, WavePlayer_23], properties {weights [1,1,1]}}
 {id=WavePlayer_0, properties {soundWaveAssetPtr=/Game/Sounds/Fort_Weapons/Impacts/Bullet_Impacts_S10/Bullet_Impact_Water_Med_10}}
 {id=WavePlayer_1, properties {soundWaveAssetPtr=/Game/Sounds/Fort_Weapons/Impacts/Bullet_Impacts_S10/Bullet_Impact_Water_Med_01}}
 {id=WavePlayer_13, properties {soundWaveAssetPtr=/Game/Sounds/Fort_Weapons/pistol/FlareGun/FlareGun_EmberImpact_Water_01}}
 {id=WavePlayer_14, properties {soundWaveAssetPtr=/Game/Sounds/Fort_Weapons/pistol/FlareGun/FlareGun_EmberImpact_Water_02}}
 {id=WavePlayer_15, properties {soundWaveAssetPtr=/Game/Sounds/Fort_Weapons/pistol/FlareGun/FlareGun_EmberImpact_Water_03}}
 {id=WavePlayer_16, properties {soundWaveAssetPtr=/Game/Sounds/Fort_Weapons/pistol/FlareGun/FlareGun_EmberImpact_Water_04}}
 {id=WavePlayer_17, properties {soundWaveAssetPtr=/Game/Sounds/Fort_Weapons/pistol/FlareGun/FlareGun_EmberImpact_Water_05}}
 {id=WavePlayer_18, properties {soundWaveAssetPtr=&FlareGun_EmberImpact_Water_01}}
 {id=WavePlayer_19, properties {soundWaveAssetPtr=/Game/Sounds/Fort_Weapons/Impacts/Bullet_Impacts_S10/Bullet_Impact_Water_Med_08}}
 {id=WavePlayer_2, properties {soundWaveAssetPtr=/Game/Sounds/Fort_Weapons/Impacts/Bullet_Impacts_S10/Bullet_Impact_Water_Med_02}}
 {id=WavePlayer_20, properties {soundWaveAssetPtr=/Game/Sounds/Fort_Weapons/Impacts/Bullet_Impacts_S10/Bullet_Impact_Water_Med_09}}
 {id=WavePlayer_21, properties {soundWaveAssetPtr=&Bullet_Impact_Water_Med_01}}
 {id=WavePlayer_22, properties {soundWaveAssetPtr=&FlareGun_EmberImpact_Water_02}}
 {id=WavePlayer_23, properties {soundWaveAssetPtr=&FlareGun_EmberImpact_Water_03}}
 {id=WavePlayer_24, properties {soundWaveAssetPtr=/Game/Sounds/Fort_Weapons/Impacts/Bullet_Impacts_S10/Bullet_Impact_Water_Med_03}}
 {id=WavePlayer_25, properties {soundWaveAssetPtr=/Game/Sounds/Fort_Weapons/Impacts/Bullet_Impacts_S10/Bullet_Impact_Water_Med_06}}
 {id=WavePlayer_3, properties {soundWaveAssetPtr=&Bullet_Impact_Water_Med_03}}
 {id=WavePlayer_4, properties {soundWaveAssetPtr=/Game/Sounds/Fort_Weapons/Impacts/Bullet_Impacts_S10/Bullet_Impact_Water_Med_04}}
 {id=WavePlayer_5, properties {soundWaveAssetPtr=/Game/Sounds/Fort_Weapons/Impacts/Bullet_Impacts_S10/Bullet_Impact_Water_Med_05}}
 {id=WavePlayer_6, properties {soundWaveAssetPtr=&Bullet_Impact_Water_Med_06}}
 {id=WavePlayer_7, properties {soundWaveAssetPtr=/Game/Sounds/Fort_Weapons/Impacts/Bullet_Impacts_S10/Bullet_Impact_Water_Med_07}}
]
properties {
 attenuationSettings=/Game/Sounds/Attenuation/Fort_Bullet_Impact_Att
 bOverrideConcurrency=true
 maxDistance=5300
 pitchMultiplier=1.1
 soundClassObject=&Fort_Impacts
 volumeMultiplier=2
 concurrencyOverrides {maxCount=5, resolutionRule=StopOldest, volumeScale=0.5}
 concurrencySet [&RifleBulletImpact_Conc]
}
