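# Sound cue for alien laser impacts, playing a random impact layered with a half-volume spark.
# A quality switch uses the full wave pools for its first two levels, and smaller pools with higher pitch ranges for the last two.
# Each layer gets random pitch modulation.
# Overridden attenuation adds distance low-pass filtering, and overridden concurrency caps simultaneous plays by stopping the oldest, under the Fort_Impacts sound class.
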SoundCue /WildEstateWeapons/Gameplay/CoveSpice/PH/Sounds/AlienLaser/AlienLaser_Impact_Cue
root=QualityLevel_0
nodes [
 {id=Mixer_0, children [Modulator_1,Modulator_0], properties {inputVolume [1,0.5]}}
 {id=Mixer_11, children [Modulator_2,Modulator_3], properties {inputVolume [1,0.5]}}
 {id=Modulator_0, children [Random_2], properties {pitchMax=1.3, pitchMin=1.2}}
 {id=Modulator_1, children [Random_3], properties {pitchMax=1.3, pitchMin=1.2}}
 {id=Modulator_2, children [Random_5], properties {pitchMax=1.1, pitchMin=0.9}}
 {id=Modulator_3, children [Random_0], properties {pitchMax=1.1, pitchMin=0.9}}
 {id=QualityLevel_0, children [Mixer_11,Mixer_11,Mixer_0,Mixer_0]}
 {
  id=Random_0
  children [WavePlayer_5, WavePlayer_6, WavePlayer_7, WavePlayer_8, WavePlayer_9, WavePlayer_10]
  properties {weights [1,1,1,1,1,1]}
 }
 {id=Random_2, children [WavePlayer_11, WavePlayer_12, WavePlayer_13], properties {weights [1,1,1]}}
 {id=Random_3, children [WavePlayer_24, WavePlayer_25, WavePlayer_26, WavePlayer_0, WavePlayer_1], properties {weights [1,1,1,1,1]}}
 {
  id=Random_5
  children [
   WavePlayer_27
   WavePlayer_28
   WavePlayer_29
   WavePlayer_30
   WavePlayer_31
   WavePlayer_32
   WavePlayer_33
   WavePlayer_34
   WavePlayer_35
   WavePlayer_36
   WavePlayer_37
   WavePlayer_38
  ]
  properties {weights [1,1,1,1,1,1,1,1,1,1,1,1]}
 }
 {id=WavePlayer_0, properties {soundWaveAssetPtr=&/AlienLaser_Impact_01}}
 {id=WavePlayer_1, properties {soundWaveAssetPtr=&/AlienLaser_Impact_02}}
 {id=WavePlayer_10, properties {soundWaveAssetPtr=&/AlienLaser_sparkbig_06}}
 {id=WavePlayer_11, properties {soundWaveAssetPtr=&/AlienLaser_sparkbig_01}}
 {id=WavePlayer_12, properties {soundWaveAssetPtr=&/AlienLaser_sparkbig_02}}
 {id=WavePlayer_13, properties {soundWaveAssetPtr=&/AlienLaser_sparkbig_03}}
 {id=WavePlayer_24, properties {soundWaveAssetPtr=&/AlienLaser_Impact_10}}
 {id=WavePlayer_25, properties {soundWaveAssetPtr=&/AlienLaser_Impact_11}}
 {id=WavePlayer_26, properties {soundWaveAssetPtr=&/AlienLaser_Impact_12}}
 {id=WavePlayer_27, properties {soundWaveAssetPtr=&AlienLaser_Impact_10}}
 {id=WavePlayer_28, properties {soundWaveAssetPtr=&AlienLaser_Impact_11}}
 {id=WavePlayer_29, properties {soundWaveAssetPtr=&AlienLaser_Impact_12}}
 {id=WavePlayer_30, properties {soundWaveAssetPtr=&AlienLaser_Impact_01}}
 {id=WavePlayer_31, properties {soundWaveAssetPtr=&AlienLaser_Impact_02}}
 {id=WavePlayer_32, properties {soundWaveAssetPtr=&/AlienLaser_Impact_03}}
 {id=WavePlayer_33, properties {soundWaveAssetPtr=&/AlienLaser_Impact_04}}
 {id=WavePlayer_34, properties {soundWaveAssetPtr=&/AlienLaser_Impact_05}}
 {id=WavePlayer_35, properties {soundWaveAssetPtr=&/AlienLaser_Impact_06}}
 {id=WavePlayer_36, properties {soundWaveAssetPtr=&/AlienLaser_Impact_07}}
 {id=WavePlayer_37, properties {soundWaveAssetPtr=&/AlienLaser_Impact_08}}
 {id=WavePlayer_38, properties {soundWaveAssetPtr=&/AlienLaser_Impact_09}}
 {id=WavePlayer_5, properties {soundWaveAssetPtr=&AlienLaser_sparkbig_01}}
 {id=WavePlayer_6, properties {soundWaveAssetPtr=&AlienLaser_sparkbig_02}}
 {id=WavePlayer_7, properties {soundWaveAssetPtr=&AlienLaser_sparkbig_03}}
 {id=WavePlayer_8, properties {soundWaveAssetPtr=&/AlienLaser_sparkbig_04}}
 {id=WavePlayer_9, properties {soundWaveAssetPtr=&/AlienLaser_sparkbig_05}}
]
properties {
 bOverrideAttenuation=true
 bOverrideConcurrency=true
 maxDistance=2600
 pitchMultiplier=1.1
 soundClassObject=&Fort_Impacts
 volumeMultiplier=1.8
 attenuationOverrides {
  bAttenuateWithLPF=true
  falloffDistance=2400
  lPFFrequencyAtMax=2000
  lPFRadiusMax=2400
  lPFRadiusMin=800
  stereoSpread=0
  attenuationShapeExtents {x=200}
 }
 concurrencyOverrides {maxCount=5, resolutionRule=StopOldest, volumeScale=0.75}
}
