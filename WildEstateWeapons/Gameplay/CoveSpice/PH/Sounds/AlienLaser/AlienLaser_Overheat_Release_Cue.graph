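# Sound cue for the AlienLaser weapon's overheat release, routed to a third-person ranged weapon fire sound class.
# A mixer layers a random pick of three numbered OverHeatLoad variants, pitched up, with the base OverHeatLoad wave, pitched down, quieter and delayed by a fixed interval.
# Close projectile attenuation is referenced, and the concurrency override blocks new instances once the limit is reached.
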SoundCue /WildEstateWeapons/Gameplay/CoveSpice/PH/Sounds/AlienLaser/AlienLaser_Overheat_Release_Cue
root=Mixer_0
nodes [
 {id=Delay_0, children [WavePlayer_0], properties {delayMax=1.4, delayMin=1.4}}
 {id=Mixer_0, children [Modulator_0,Modulator_1], properties {inputVolume [1,0.6]}}
 {id=Modulator_0, children [Random_0], properties {pitchMax=1.1, pitchMin=1.1}}
 {id=Modulator_1, children [Delay_0], properties {pitchMax=0.9, pitchMin=0.9}}
 {id=Random_0, children [WavePlayer_1,WavePlayer_2,WavePlayer_3], properties {weights [1,1,1]}}
 {id=WavePlayer_0, properties {soundWaveAssetPtr=&/AlienLaser_OverHeatLoad}}
 {id=WavePlayer_1, properties {soundWaveAssetPtr=&/AlienLaser_OverHeatLoad_01}}
 {id=WavePlayer_2, properties {soundWaveAssetPtr=&/AlienLaser_OverHeatLoad_02}}
 {id=WavePlayer_3, properties {soundWaveAssetPtr=&/AlienLaser_OverHeatLoad_03}}
]
properties {
 attenuationSettings=&Fort_Weapon_Projectile_Close_Athena_Att
 bOverrideConcurrency=true
 maxDistance=4950
 soundClassObject=&Fort_WeaponRanged_Fire_3P_PSM
 volumeMultiplier=0.7
 attenuationOverrides {distanceAlgorithm=NaturalSound, falloffDistance=4000, attenuationShapeExtents {x=1000}}
 concurrencyOverrides {resolutionRule=PreventNew}
}
